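# Physics asset for the GuanYu two-handed pickaxe cosmetic, giving it a dangling bell on a jointed loop.
# The handle body combines a long shaft capsule with two short spine collision capsules.
# Constraints chain the handle to a loop center sphere, a loop pitch capsule, then the bell sphere.
# Each joint locks twist and one swing axis, so the loop center swings on swing 2 and the two outer links on swing 1.
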PhysicsAsset /BRCosmetics/Weapons/FORT_Melee/Pickaxe_GuanYu/Meshes/Pickaxe_GuanYu_Physics
bodies {
 Melee_TwoHanded_Handle [
  {
   shapeType=Capsule
   length=6.629
   radius=2.7487
   shapeName=LowerSpineCollision
   center=<0.8074 -5.314 79.7867>
   rotation=<rot -25.6141 -89.9998 0.0195>
  }
  {shapeType=Capsule, length=7.75, radius=2, shapeName=UpperSpineCollision, center=<0.6797 -15.7177 92.6314>, rotation=<rot 0 0 -28.5489>}
  {shapeType=Capsule, length=163.25, radius=4.25, center=<0 0 33.3207>}
 ]
 dyn_Bell [{shapeType=Sphere, radius=4.25, shapeName=Bell, center=<10.2256 -0.0136 0.1553>}]
 dyn_LoopCenter [{shapeType=Sphere, radius=1.755, shapeName=RingTwist}]
 dyn_LoopPitch [{shapeType=Capsule, length=5, radius=1.25, shapeName=RingPitch, center=<7.4596 0 -0.0001>, rotation=<rot 0 -90 0>}]
}
constraints [
 {
  bone1Name=dyn_LoopCenter
  bone2Name=Melee_TwoHanded_Handle
  swing1LimitDegrees=45
  swing1Motion=Locked
  swing2LimitDegrees=45
  twistLimitDegrees=45
  twistMotion=Locked
 }
 {
  bone1Name=dyn_LoopPitch
  bone2Name=dyn_LoopCenter
  swing1LimitDegrees=15
  swing2LimitDegrees=45
  swing2Motion=Locked
  twistLimitDegrees=45
  twistMotion=Locked
 }
 {
  bone1Name=dyn_Bell
  bone2Name=dyn_LoopPitch
  swing1LimitDegrees=45
  swing2LimitDegrees=45
  swing2Motion=Locked
  twistLimitDegrees=45
  twistMotion=Locked
 }
]
